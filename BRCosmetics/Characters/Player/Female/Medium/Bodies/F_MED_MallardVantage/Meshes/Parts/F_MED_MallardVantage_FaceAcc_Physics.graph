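# Physics asset for the MallardVantage female medium cosmetic's face accessory, setting up jiggle bones for helmet, straps and necklace tags.
# All bodies are capsules, with head, neck and spine_05 anchoring the dyn_ chains.
# The helmet carries a band, card and two four-link straps; a four-link necklace from spine_05 carries two tags.
# Chain links use tight swing/twist limits, while MSTR joints swing and twist freely, except the locked right-strap root.
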PhysicsAsset /BRCosmetics/Characters/Player/Female/Medium/Bodies/F_MED_MallardVantage/Meshes/Parts/F_MED_MallardVantage_FaceAcc_Physics
bodies {
 dyn_band_01 [{shapeType=Capsule, length=5, radius=3, shapeName=dyn_band_01_capsule, center=<1 0 0>, rotation=<rot 90 0 0>}]
 dyn_card_01 [{shapeType=Capsule, length=5, radius=3, shapeName=dyn_card_01_capsule, center=<1 0 0>, rotation=<rot 90 0 0>}]
 dyn_helmet_01 [{shapeType=Capsule, length=5, radius=3, shapeName=dyn_helmet_01_capsule, center=<1 0 0>, rotation=<rot 90 0 0>}]
 dyn_helmet_MSTR [{shapeType=Capsule, length=0.505, radius=0.505, shapeName=dyn_helmet_MSTR_capsule}]
 dyn_necklace_01 [{shapeType=Capsule, length=8, radius=3, shapeName=dyn_necklace_01_capsule, center=<1 2 0>}]
 dyn_necklace_02 [{shapeType=Capsule, length=8, radius=3, shapeName=dyn_necklace_02_capsule, center=<1 2 0>}]
 dyn_necklace_03 [{shapeType=Capsule, length=8, radius=3, shapeName=dyn_necklace_03_capsule, center=<1 2 0>}]
 dyn_necklace_04 [{shapeType=Capsule, length=8, radius=3, shapeName=dyn_necklace_04_capsule, center=<1 2 0>, rotation=<rot 0 0.0001 -5>}]
 dyn_necklace_MSTR [{shapeType=Capsule, length=5, radius=3, shapeName=dyn_necklace_MSTR_capsule, center=<1 2 0>}]
 dyn_strap_l_01 [{shapeType=Capsule, length=5, radius=3, shapeName=dyn_strap_l_01_capsule, center=<2 0 0>, rotation=<rot 90 0 0>}]
 dyn_strap_l_02 [{shapeType=Capsule, length=5, radius=3, shapeName=dyn_strap_l_02_capsule, center=<2 0 0>, rotation=<rot 90 0 0>}]
 dyn_strap_l_03 [{shapeType=Capsule, length=5, radius=3, shapeName=dyn_strap_l_03_capsule, center=<2 0 0>, rotation=<rot 90 0 0>}]
 dyn_strap_l_04 [{shapeType=Capsule, length=5, radius=3, shapeName=dyn_strap_l_04_capsule, center=<2 0 0>, rotation=<rot 90 0 0>}]
 dyn_strap_r_01 [{shapeType=Capsule, length=5, radius=3, shapeName=dyn_strap_r_01_capsule, center=<2 0 0>, rotation=<rot 90 0 0>}]
 dyn_strap_r_02 [{shapeType=Capsule, length=5, radius=3, shapeName=dyn_strap_r_02_capsule, center=<2 0 0>, rotation=<rot 90 0 0>}]
 dyn_strap_r_03 [{shapeType=Capsule, length=5, radius=3, shapeName=dyn_strap_r_03_capsule, center=<2 0 0>, rotation=<rot 90 0 0>}]
 dyn_strap_r_04 [{shapeType=Capsule, length=5, radius=3, shapeName=dyn_strap_r_04_capsule, center=<2 0 0>, rotation=<rot 90 0 0>}]
 dyn_strap_r_MSTR [{shapeType=Capsule, length=0.505, radius=0.505, shapeName=dyn_strap_r_MSTR_capsule}]
 dyn_tag_01 [{shapeType=Capsule, length=3, radius=2, shapeName=dyn_tag_01_capsule, center=<1 0 0>, rotation=<rot 90 3 0>}]
 dyn_tag_02 [{shapeType=Capsule, length=3, radius=2, shapeName=dyn_tag_02_capsule, center=<1 0 0>, rotation=<rot 90 3 0>}]
 dyn_tag_MSTR [{shapeType=Capsule, length=2.5, radius=2.5, shapeName=dyn_tag_MSTR_capsule, center=<1 0 0>, rotation=<rot 0 3 0>}]
 head [{shapeType=Capsule, length=15, radius=6, shapeName=head_capsule, center=<6.0854 0 0>, rotation=<rot 90 0 0>}]
 neck_01 [{shapeType=Capsule, length=0.505, radius=0.505, shapeName=neck_01_capsule}]
 neck_02 [{shapeType=Capsule, length=0.505, radius=0.505, shapeName=neck_02_capsule}]
 spine_05 [{shapeType=Capsule, length=7, radius=19, shapeName=spine_05_capsule, center=<-12.2206 -1.0947 0>, rotation=<rot 90 0 4.3533>}]
}
constraints [
 {bone1Name=dyn_strap_r_04, bone2Name=dyn_strap_r_03, swing1LimitDegrees=15, swing2LimitDegrees=5, twistLimitDegrees=1}
 {bone1Name=neck_02, bone2Name=neck_01, swing1LimitDegrees=15, swing2LimitDegrees=15, twistLimitDegrees=1}
 {bone1Name=head, bone2Name=neck_02, swing1LimitDegrees=15, swing2LimitDegrees=15, twistLimitDegrees=1}
 {
  bone1Name=dyn_helmet_MSTR
  bone2Name=head
  swing1LimitDegrees=45
  swing1Motion=Free
  swing2LimitDegrees=45
  swing2Motion=Free
  twistLimitDegrees=45
  twistMotion=Free
 }
 {bone1Name=dyn_helmet_01, bone2Name=dyn_helmet_MSTR, swing1LimitDegrees=4, swing2LimitDegrees=2, twistLimitDegrees=1}
 {bone1Name=dyn_band_01, bone2Name=dyn_helmet_01, swing1LimitDegrees=15, swing2LimitDegrees=1, twistLimitDegrees=1}
 {bone1Name=dyn_card_01, bone2Name=dyn_helmet_01, swing1LimitDegrees=8, swing2LimitDegrees=1, twistLimitDegrees=1}
 {bone1Name=dyn_strap_l_01, bone2Name=dyn_helmet_01, swing1LimitDegrees=30, swing2LimitDegrees=5, twistLimitDegrees=1}
 {bone1Name=dyn_strap_l_02, bone2Name=dyn_strap_l_01, swing1LimitDegrees=15, swing2LimitDegrees=5, twistLimitDegrees=4}
 {bone1Name=dyn_strap_l_03, bone2Name=dyn_strap_l_02, swing1LimitDegrees=15, swing2LimitDegrees=5, twistLimitDegrees=4}
 {bone1Name=dyn_strap_l_04, bone2Name=dyn_strap_l_03, swing1LimitDegrees=15, swing2LimitDegrees=5, twistLimitDegrees=1}
 {
  bone1Name=dyn_strap_r_MSTR
  bone2Name=dyn_helmet_01
  swing1LimitDegrees=45
  swing1Motion=Locked
  swing2LimitDegrees=45
  swing2Motion=Locked
  twistLimitDegrees=45
  twistMotion=Locked
 }
 {bone1Name=dyn_strap_r_01, bone2Name=dyn_strap_r_MSTR, swing1LimitDegrees=30, swing2LimitDegrees=5, twistLimitDegrees=1}
 {bone1Name=dyn_strap_r_02, bone2Name=dyn_strap_r_01, swing1LimitDegrees=15, swing2LimitDegrees=5, twistLimitDegrees=4}
 {bone1Name=dyn_strap_r_03, bone2Name=dyn_strap_r_02, swing1LimitDegrees=15, swing2LimitDegrees=5, twistLimitDegrees=4}
 {
  bone1Name=neck_01
  bone2Name=spine_05
  swing1LimitDegrees=45
  swing1Motion=Free
  swing2LimitDegrees=45
  swing2Motion=Free
  twistLimitDegrees=45
  twistMotion=Free
 }
 {bone1Name=dyn_tag_01, bone2Name=dyn_tag_MSTR, swing1LimitDegrees=5, swing2LimitDegrees=30, twistLimitDegrees=2}
 {bone1Name=dyn_tag_02, bone2Name=dyn_tag_MSTR, swing1LimitDegrees=5, swing2LimitDegrees=30, twistLimitDegrees=2}
 {
  bone1Name=dyn_tag_MSTR
  bone2Name=dyn_necklace_04
  swing1LimitDegrees=45
  swing1Motion=Free
  swing2LimitDegrees=45
  swing2Motion=Free
  twistLimitDegrees=45
  twistMotion=Free
 }
 {
  bone1Name=dyn_necklace_MSTR
  bone2Name=spine_05
  swing1LimitDegrees=45
  swing1Motion=Free
  swing2LimitDegrees=45
  swing2Motion=Free
  twistLimitDegrees=45
  twistMotion=Free
 }
 {bone1Name=dyn_necklace_01, bone2Name=dyn_necklace_MSTR, swing1LimitDegrees=7, swing2LimitDegrees=2, twistLimitDegrees=1}
 {bone1Name=dyn_necklace_02, bone2Name=dyn_necklace_01, swing1LimitDegrees=7, swing2LimitDegrees=2, twistLimitDegrees=1}
 {bone1Name=dyn_necklace_03, bone2Name=dyn_necklace_02, swing1LimitDegrees=7, swing2LimitDegrees=2, twistLimitDegrees=1}
 {bone1Name=dyn_necklace_04, bone2Name=dyn_necklace_03, swing1LimitDegrees=7, swing2LimitDegrees=2, twistLimitDegrees=1}
]
